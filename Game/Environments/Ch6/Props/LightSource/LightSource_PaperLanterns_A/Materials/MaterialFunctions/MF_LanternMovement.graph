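# Material function that animates paper lanterns with world position offset and corrected normals, behind a static switch with a Swing Only variant.
# It combines rope swing, wind-noise-driven lantern swing and twist rotations, using position-derived random time to vary motion.
# Local X distance picks among three pivots mirrored per side, while UV and position masks limit deformation.
# Custom primitive data supplies a hanging toggle, and distance fades the offset, with a separate Nanite path.
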MaterialFunction /Game/Environments/Ch6/Props/LightSource/LightSource_PaperLanterns_A/Materials/MaterialFunctions/MF_LanternMovement
graph {nodes [
  n1 = ActorPositionWS()
  n2 = NamedRerouteUsage()
  Add(A=@n1, B=@n2)
  Floor(None=Divide() {constB=100})
  Fmod()
  n3 = FunctionInput() {inputType=FunctionInput_MaterialAttributes}
  n4 = GetMaterialAttributes(None=@n3) {attributeGetTypes [#1, #2]}
  n5 = NamedRerouteUsage()
  n6 = Transform(None=Add(A=NamedRerouteUsage(), B=NamedRerouteUsage()) {constB=0}) {transformSourceType=TRANSFORMSOURCE_Local}
  n7 = NamedRerouteUsage()
  n8 = StaticSwitchParameter(True=@n5, False=Add(A=@n5, B=LinearInterpolate(A=@n6, Alpha=@n7) {constB=0})) {parameterName=Swing Only}
  n9 = NamedRerouteUsage()
  n10 = Add(A=@n4.WorldPositionOffset, B=Multiply(A=LinearInterpolate(A=@n8, Alpha=@n9) {constB=0}))
  n11 = MaterialFunctionCall(Rotation Axis=Constant3Vector() {constant=<rgba 1 0 0 0>}, Rotation Angle=NamedRerouteUsage(), World Space Vertex Normals=@n4.Normal) {materialFunction=&FixRotateAboutAxisNormals}
  FunctionOutput(None=StaticSwitchParameter(
    True=SetMaterialAttributes!(MaterialAttributes=@n4.MaterialAttributes, WorldPositionOffset=@n10, Normal=@n11.New World Space Vertex Normal, EmissiveColor) {attributeSetTypes [#1, #2, #3]}
    False=@n3
   ) {defaultValue=true, parameterName=Use Lantern Movement})
  n12 = Constant3Vector()
  n13 = Divide(
   A=Ceil(None=TransformPosition(None=@n12) {transformType=TRANSFORMPOSSOURCE_World})
   B=ScalarParameter() {defaultValue=10000, parameterName=LeapOfFaith}
  )
  n14 = MaterialFunctionCall(Float3=Multiply(A=Add(A=@n13, B=@n2)) {constB=11.173}) {materialFunction=&BreakOutFloat3Components}
  n15 = Subtract(A=@n14.R, B=@n14.R)
  n16 = LinearInterpolate(Alpha=Frac(None=@n15)) {constA=0.97, constB=1.03}
  MaterialFunctionCall!(Time, Divisor=@n16, Mod=ScalarParameter() {defaultValue=10, parameterName=FMod}) {materialFunction=/Game/Effects/Fort_Effects/Materials/_Functions/MF_ModuloTime}
  Modulo!(A=@n1, B=Constant() {r=3000})
  n17 = Time() {period=86400}
  Multiply(A=@n17, B=@n16)
  Multiply(A=@n15)
  n18 = If(
   A=ComponentMask(None=TextureCoordinate()) {g=true}
   B=ScalarParameter() {defaultValue=0.037, parameterName=UV Mask Threshold}
   "A > B"=Constant() {r=1}
   "A < B"=Constant()
  )
  n19 = LocalPosition()
  n20 = If(
   A=Abs(None=ComponentMask(None=@n19.XYZ) {r=true})
   B=ScalarParameter() {defaultValue=254, parameterName=Local Position X Mask}
   "A > B"=Constant()
   "A < B"=Constant() {r=1}
  )
  NamedRerouteDeclaration(None=LinearInterpolate(A=Multiply(A=@n18, B=@n20), Alpha=ScalarParameter() {parameterName=UseCharacterLantern})) {name=WPO Mask, nodeColor=<rgba 0.0781 0 1 1>}
  n21 = Normalize(VectorInput=NamedRerouteUsage())
  n22 = Sine(None=Multiply(A=NamedRerouteUsage(), B=ScalarParameter() {defaultValue=0.4, parameterName=Swing Speed}) {constB=0.5})
  n23 = ComponentMask(None=NamedRerouteUsage()) {r=true}
  n24 = Subtract(
   A=Multiply(A=LinearInterpolate(A=@n22, Alpha=@n23) {constAlpha=0}, B=ScalarParameter() {defaultValue=0.0075, parameterName=Wind Strength}) {constB=0.01}
   B=NamedRerouteUsage()
  )
  n25 = NamedRerouteUsage()
  n26 = NamedRerouteUsage()
  n27 = Add(
   A=LocalPosition().XYZ
   B=Add(A=NamedRerouteUsage(), B=Transform(None=@n26) {transformSourceType=TRANSFORMSOURCE_World, transformType=TRANSFORM_Local})
  )
  NamedRerouteDeclaration(None=Multiply(A=RotateAboutAxis(NormalizedRotationAxis=@n21, RotationAngle=@n24, PivotPoint=@n25, Position=@n27), B=NamedRerouteUsage())) {name=Swing WPO, nodeColor=<rgba 0.2422 1 0 1>}
  n28 = ScalarParameter() {defaultValue=5000, group=WPO Fade, parameterName=WPO_Fade_Distance}
  n29 = Divide(A=Subtract(A=PixelDepth().R, B=ScalarParameter() {defaultValue=5000, group=WPO Fade, parameterName=WPO_Fade_Offset}), B=@n28)
  n30 = MaterialFunctionCall!(FadeBlendDistance=@n28, AutoHandleBadDistance) {materialFunction=/Game/Global/MaterialFunctions/TechArt/MF_WPODisableBlend}
  NamedRerouteDeclaration(None=Saturate(None=NaniteReplace!(Default=@n29, Nanite=@n30.VertexShaderOutput))) {name=Distance WPO Fade, nodeColor=<rgba 1 0 0.9844 1>}
  NamedRerouteDeclaration(None=ScalarParameter() {bUseCustomPrimitiveData=true, parameterName=Use Hanging, primitiveDataIndex=3}) {name=Use Hanging, nodeColor=<rgba 1 0.7969 0 1>}
  n31 = NamedRerouteUsage()
  n32 = Add(A=LocalPosition().XYZ, B=Transform(None=NamedRerouteUsage()) {transformSourceType=TRANSFORMSOURCE_World, transformType=TRANSFORM_Local})
  n33 = Multiply(A=NamedRerouteUsage(), B=ScalarParameter() {defaultValue=0.1, parameterName=Twisting Speed}) {constB=0.2}
  n34 = ComponentMask(None=NamedRerouteUsage()) {r=true}
  n35 = Multiply(
   A=LinearInterpolate(B=Cosine(None=@n33), Alpha=ConstantBiasScale(None=@n34) {bias=-0.5, scale=2})
   B=ScalarParameter() {defaultValue=0.015, parameterName=Diveation Amount}
  )
  n36 = Constant3Vector() {constant=<rgba 0 0 1 0>}
  NamedRerouteDeclaration(None=Multiply(A=RotateAboutAxis(NormalizedRotationAxis=@n36, RotationAngle=@n35, PivotPoint=@n31, Position=@n32), B=NamedRerouteUsage())) {name=Twist WPO, nodeColor=<rgba 1 0.7734 0 1>}
  n37 = VectorParameter() {parameterName=Pivot1, defaultValue=<rgba 0 0 0 1>}
  n38 = ComponentMask(None=LocalPosition() {includedOffsets=ExcludeOffsets}.XYZ) {r=true}
  n39 = Constant() {r=1}
  n40 = MaterialFunctionCall(X=Sign(None=@n38), Y=@n39, Z=@n39) {materialFunction=&MakeFloat3}
  n41 = Multiply(A=VectorParameter() {parameterName=Pivot 2, defaultValue=<rgba 106.5 0 3 1>}.RGB, B=@n40.Result)
  n42 = Multiply(A=VectorParameter() {parameterName=Pivot 3, defaultValue=<rgba 208.7 0 15 1>}.RGB, B=@n40.Result)
  n43 = Abs(None=@n38)
  n44 = Floor(None=Divide(A=@n43, B=ScalarParameter() {defaultValue=55.7112, parameterName=AwayFromCenterMask1}))
  n45 = Divide(A=@n43, B=ScalarParameter() {defaultValue=156.3277, parameterName=AwayFromCenterMask2})
  n46 = Max(A=Multiply(A=Saturate(None=@n44)) {constB=0.5}, B=Saturate(None=Floor(None=@n45)))
  NamedRerouteDeclaration(None=LinearInterpolate(
    A=MaterialFunctionCall(1=@n37.RGB, 2=@n41, 3=@n42, A=@n46) {materialFunction=&Lerp_Multiple_Float3}.Lerp 3 Inputs
    Alpha=NamedRerouteUsage()
   ) {constB=0}) {name=Pivots, nodeColor=<rgba 0 1 0.7891 1>}
  n47 = TransformPosition(None=NamedRerouteUsage()) {transformType=TRANSFORMPOSSOURCE_World}
  n48 = Panner(
   Coordinate=Divide(A=ComponentMask(None=@n47) {g=true, r=true}, B=ScalarParameter() {defaultValue=50, parameterName=Wind Scale})
   Time=Time()
   Speed=MaterialFunctionCall(X=ScalarParameter() {defaultValue=0.03, parameterName=Wind Speed X}, Y=ScalarParameter() {defaultValue=0.01, parameterName=Wind Speed Y}) {materialFunction=&MakeFloat2}.Result
  )
  NamedRerouteDeclaration(None=ConstantBiasScale(None=MaterialFunctionCall!(UV1=@n48, Min, Max) {materialFunction=/Game/Athena/WinterArt/Materials/MF_SimpleWindNoise}.Result) {bias=-0.5, scale=2}) {name=Wind, nodeColor=<rgba 0.5937 1 0 1>}
  n49 = Constant()
  n50 = Constant3Vector() {constant=<rgba 1 0 0 0>}
  NamedRerouteDeclaration(None=Add(A=RotateAboutAxis(NormalizedRotationAxis=@n36, RotationAngle=@n35, PivotPoint=@n49, Position=@n50), B=@n50)) {name=Twist Axis, nodeColor=<rgba 0.4531 0 1 1>}
  n51 = Frac(None=Add(A=@n14.R, B=@n14.R))
  NamedRerouteDeclaration(None=Add(A=Multiply(A=@n51) {constB=120}, B=@n17)) {name=Random TIme, nodeColor=<rgba 1 0.7266 0 1>}
  n52 = Constant3Vector() {constant=<rgba 1 0 0 0>}
  n53 = Transform(None=@n52) {transformSourceType=TRANSFORMSOURCE_Local}
  n54 = Sine(None=Multiply(A=NamedRerouteUsage(), B=ScalarParameter() {defaultValue=0.025, parameterName=Deviation Speed}))
  n55 = Normalize(VectorInput=MaterialFunctionCall(1=@n52, 2=Constant3Vector() {constant=<rgba 0 1 0 0>}, 3=Constant3Vector() {constant=<rgba -1 0 0 0>}, A=ConstantBiasScale(None=@n54)) {materialFunction=&Lerp_Multiple_Float3}.Lerp 3 Inputs)
  n56 = NamedRerouteUsage()
  n57 = NamedRerouteUsage()
  n58 = ScalarParameter() {defaultValue=0.2, parameterName=Rope Swing Speed}
  n59 = Multiply(A=Sine(None=Multiply(A=@n57, B=@n58)), B=ScalarParameter() {defaultValue=0.015, parameterName=Rope Swing Amount})
  n60 = ActorPositionWS() {originType=CameraRelative}
  n61 = LinearInterpolate(
   A=VectorParameter() {parameterName=Rope Pivot Offset, defaultValue=<rgba 0 0 26 0>}.RGB
   B=VectorParameter() {parameterName=Swing Rope Pivot Offset}.RGB
   Alpha=NamedRerouteUsage()
  )
  NamedRerouteDeclaration(None=RotateAboutAxis(
    NormalizedRotationAxis=LinearInterpolate(A=@n53, B=@n55, Alpha=@n56)
    RotationAngle=NamedRerouteDeclaration(None=@n59) {name=Rope Swing Amount, nodeColor=<rgba 0 1 0.1328 1>}
    PivotPoint=Add(A=@n60, B=@n61)
    Position=WorldPosition() {worldPositionShaderOffset=WPT_CameraRelative}.XYZ
   )) {name=Rope Swing, nodeColor=<rgba 1 0 0.4219 1>}
  NamedRerouteDeclaration(None=@n46) {nodeColor=<rgba 1 0 0.6797 1>}
  NamedRerouteUsage()
  NamedRerouteUsage()
  NamedRerouteUsage()
  n62 = MaterialFunctionCall(Float3=@n47) {materialFunction=&BreakOutFloat3Components}
  n63 = Add(A=@n62.R, B=@n62.R)
  Sine(None=Add(A=Time(), B=Add(A=@n63, B=@n62.B)))
 ]}
